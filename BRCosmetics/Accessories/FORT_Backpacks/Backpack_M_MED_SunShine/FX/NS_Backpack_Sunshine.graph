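# Niagara effect for the Sunshine backpack cosmetic, combining glow, smoke puffs, dry-ice vapour and snow flecks.
# LightAndGlow and idle vapour spawn at User.CenterOffset; smoke, exhaust and snow spawn from the skeletal mesh, rise and partly inherit owner movement.
# Curves over system loop age pulse spawn rates and fades, while owner speed fades both dry-ice layers.
# User.VariantIndex blends glow and smoke from warm orange to icy blue, with time-of-day brightness adjustments.
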NiagaraSystem /BRCosmetics/Accessories/FORT_Backpacks/Backpack_M_MED_SunShine/FX/NS_Backpack_Sunshine
dependencies {
 usedDataInterfaces [
  &NiagaraDataInterfaceCurve
  &NiagaraDataInterfacePlatformSet
  &NiagaraDataInterfaceArrayFloat3
  &NiagaraDataInterfaceSkeletalMesh
 ]
 usedDynamicInputs [&SetBoolByFloatComparison]
 usedModules [
  &self:SystemSpawnScript.SystemScriptSource.SystemScriptGraph.NiagaraNodeAssignment_0.SetVariables_#1
  &self:SystemSpawnScript.SystemScriptSource.SystemScriptGraph.NiagaraNodeAssignment_1.SetVariables_#2
  &self:SystemSpawnScript.SystemScriptSource.SystemScriptGraph.NiagaraNodeAssignment_2.SetVariables_#3
  &SystemState
 ]
}
emitters {
 Dry_Ice_Idle {
  emitterUpdateScript [
   &EmitterState()
   &SpawnBurst_Instantaneous(Spawn Count=2, Use Spawn Probability=false, UseLoopCountLimit=false)
   &SpawnRate(
    SpawnRate=&FloatFromCurve(
     FloatCurve=DataInterface() {propertyValues {Curve {keys [
         {tangentMode=RCTM_User, time=0, value=1}
         {arriveTangent=-3.5689, interpMode=RCIM_Cubic, leaveTangent=-3.5689, tangentMode=RCTM_User, time=0.5, value=1}
         {arriveTangent=-1.8913, interpMode=RCIM_Cubic, leaveTangent=-1.8913, tangentMode=RCTM_User, time=0.55, value=0}
        ]}}}
     CurveIndex=$System.NormalizedLoopAge
     Scale Curve=3
    )
    Use Spawn Probability=false
   )
  ]
  particleSpawnScript [
   &InitializeParticle(
    Color=<rgba 0.992 1.4456 2 0.3>
    Lifetime=&RandomRangeFloat(Minimum=1.5, Maximum=2.5)
    Position Offset=&AddVectorToPosition(Position=$User.CenterOffset, Vector=<0 0 -2>)
    Position Offset Coordinate Space=::Local
    Sprite Rotation Mode=::NewEnumerator2
    Sprite Size Mode=::NewEnumerator3
    Sprite UV Mode=::NewEnumerator3
    Uniform Sprite Size=30
    UsePositionOffset=true
    Write Material Random=true
   )
   &AddVelocity(Velocity=&SelectVectorFromArray(
     Array Sampling Mode=::NewEnumerator0
     Direct Array Index=?
     Vector Selection Array=DataInterface() {propertyValues {FloatData [<8 0 0>, <-8 0 0>]}}
     Random Seed=0
    ))
   &CameraOffset(Camera Offset Amount=8)
  ]
  particleUpdateScript [
   &ParticleState()
   &ScaleSpriteSize(Uniform Curve Sprite Scale=DataInterface() {propertyValues {Curve {keys [{interpMode=RCIM_Cubic, tangentMode=RCTM_User, time=0, value=0.25}, {interpMode=RCIM_Cubic, tangentMode=RCTM_User, time=1, value=1}]}}})
   &ScaleColor(
    Scale Alpha=&Multiply_Float(
     A=&FloatFromCurve(FloatCurve=DataInterface() {propertyValues {Curve {keys [
          {interpMode=RCIM_Cubic, tangentMode=RCTM_User, time=0, value=0}
          {interpMode=RCIM_Cubic, tangentMode=RCTM_User, time=0.2, value=1}
          {tangentMode=RCTM_User, time=1, value=0}
         ]}}})
     B=&FloatFromCurve(
      FloatCurve=DataInterface() {propertyValues {Curve {keys [
          {interpMode=RCIM_Cubic, tangentMode=RCTM_User, time=0, value=0}
          {arriveTangent=1.6667, leaveTangent=1.6667, time=0.1, value=1}
          {arriveTangent=-1.1111, interpMode=RCIM_Cubic, leaveTangent=-1.1111, tangentMode=RCTM_User, time=0.5, value=1}
          {arriveTangent=-5.2539, interpMode=RCIM_Cubic, leaveTangent=-5.2539, tangentMode=RCTM_User, time=1, value=0}
         ]}}}
      CurveIndex=$System.NormalizedLoopAge
      Scale Curve=0.5
     )
    )
    Scale RGB=&VectorFromFloat(Value=&FNTimeOfDayScalar(Night Scale=0.8, Morning Scale=1, Evening Scale=0.8))
    ScaleA=true
    ScaleRGB=true
   )
   /Niagara/Modules/Update/Color/ScaleColorBySpeed(
    A Scale Max=0
    A Scale Min=1
    Max Speed Threshold=500
    Min Speed Threshold=0
    RGB Scale Min=1
    Source Velocity:Vector3f=$Engine.Owner.Velocity
   )
   &Drag(Use Linear Drag=true, Use Rotational Drag=true)
   &GravityForce(Gravity=&Multiply_Vector(
     A=<1 1 30>
     B=&VectorFromFloat(Value=&FloatFromCurve(FloatCurve=DataInterface() {propertyValues {Curve {keys [<at 0.2 0>,<at 1 1>]}}}))
    ))
   &SolveForcesAndVelocity(Clamp Velocity=false, Limit Acceleration=false)
  ]
  rendererClasses [&NiagaraSpriteRendererProperties]
  renderers [{rendererClass=&NiagaraSpriteRendererProperties}]
 }
 Dry_Ice_Smoke_Exhaust {
  emitterUpdateScript [
   &EmitterState()
   &SpawnRate(
    SpawnRate=&FloatFromCurve(
     FloatCurve=DataInterface() {propertyValues {Curve {keys [
         {tangentMode=RCTM_User, time=0, value=1}
         {arriveTangent=-3.5689, interpMode=RCIM_Cubic, leaveTangent=-3.5689, tangentMode=RCTM_User, time=0.6, value=1}
         {arriveTangent=-1.8913, interpMode=RCIM_Cubic, leaveTangent=-1.8913, tangentMode=RCTM_User, time=1, value=0}
        ]}}}
     CurveIndex=$System.NormalizedLoopAge
     Scale Curve=10
    )
    Use Spawn Probability=false
   )
  ]
  particleSpawnScript [
   &InitializeParticle(
    Color=<rgba 0.4375 1.1406 2 1>
    Lifetime=&RandomRangeFloat(Minimum=0.5, Maximum=1)
    Sprite Rotation Mode=::NewEnumerator2
    Sprite Size Mode=::NewEnumerator3
    Sprite UV Mode=::NewEnumerator3
    Uniform Sprite Size=30
    UsePositionOffset=false
    Write Material Random=true
   )
   &SkeletalMeshLocation(MaskSpawnGroup=false, Mesh Sampling Type=::NewEnumerator5, Skeletal Mesh=$System.Skeletal Mesh, Socket Sampling Mode=::NewEnumerator5)
   &AddVelocity(Velocity=<25 0 12>)
  ]
  particleUpdateScript [
   &ParticleState()
   &ScaleSpriteSize(Uniform Curve Sprite Scale=DataInterface() {propertyValues {Curve {keys [{interpMode=RCIM_Cubic, tangentMode=RCTM_User, time=0, value=0.25}, {interpMode=RCIM_Cubic, tangentMode=RCTM_User, time=1, value=1}]}}})
   &ScaleColor(
    Scale Alpha=&Multiply_Float(
     A=&FloatFromCurve(
      FloatCurve=DataInterface() {propertyValues {Curve {keys [
          {interpMode=RCIM_Cubic, tangentMode=RCTM_User, time=0, value=0}
          {interpMode=RCIM_Cubic, tangentMode=RCTM_User, time=0.5, value=1}
          {tangentMode=RCTM_User, time=1, value=0}
         ]}}}
      Scale Curve=0.02
     )
     B=&FloatFromCurve(
      FloatCurve=DataInterface() {propertyValues {Curve {keys [
          {interpMode=RCIM_Cubic, tangentMode=RCTM_User, time=0, value=0}
          {arriveTangent=1.6667, leaveTangent=1.6667, time=0.1, value=1}
          {arriveTangent=-1.1111, interpMode=RCIM_Cubic, leaveTangent=-1.1111, tangentMode=RCTM_User, time=1.5, value=1}
          {arriveTangent=-5.2539, interpMode=RCIM_Cubic, leaveTangent=-5.2539, tangentMode=RCTM_User, time=2, value=0}
         ]}}}
      CurveIndex=$Emitter.Age
      Scale Curve=0.5
     )
    )
    Scale RGB=&VectorFromFloat(Value=&FNTimeOfDayScalar(Night Scale=0.8, Morning Scale=1, Evening Scale=0.8))
    ScaleA=true
    ScaleRGB=true
   )
   &ScaleColorBySpeed(
    A Scale Max=0
    A Scale Min=1
    Max Speed Threshold=500
    Min Speed Threshold=0
    RGB Scale Min=1
    Source Velocity:Vector3f=$Engine.Owner.Velocity
   )
   &CameraOffset(Camera Offset Amount=10)
   &GravityForce(Gravity=&Multiply_VectorByFloat(
     Vector=<0 0 250>
     Float=&FloatFromCurve(FloatCurve=DataInterface() {propertyValues {Curve {keys [
          {interpMode=RCIM_Cubic, tangentMode=RCTM_User, time=0, value=0}
          {arriveTangent=2.3053, interpMode=RCIM_Cubic, leaveTangent=2.3053, tangentMode=RCTM_User, time=1, value=1}
         ]}}})
    ))
   &InheritSourceMovement(
    Applied Position Scale=&Multiply_VectorByFloat(
     Vector=<1 1 1>
     Float=&FloatFromCurve(FloatCurve=DataInterface() {propertyValues {Curve {keys [
          <at 0 1 RCIM_Cubic>
          {arriveTangent=-0.2, interpMode=RCIM_Cubic, leaveTangent=-0.2, time=0.25, value=1}
          {arriveTangent=-0.0269, interpMode=RCIM_Cubic, leaveTangent=-0.0269, tangentMode=RCTM_User, time=1, value=0.8}
         ]}}})
    )
    ApplyForce=false
    ApplyPosition=true
    ApplyPositionFalloff=false
   )
   &Drag(Drag=0.5, Use Linear Drag=true, Use Rotational Drag=true)
   &SolveForcesAndVelocity(Clamp Velocity=false, Limit Acceleration=false)
  ]
  rendererClasses [&NiagaraSpriteRendererProperties]
  renderers [{rendererClass=&NiagaraSpriteRendererProperties}]
 }
 LightAndGlow {
  emitterUpdateScript [
   &EmitterState()
   &SpawnBurst_Instantaneous(Spawn Count=2, Use Spawn Probability=false, UseLoopCountLimit=false)
  ]
  particleSpawnScript [
   &InitializeParticle(
    Color=&LerpLinearColors(StartColor=<rgba 4 1 0 2>, EndColor=<rgba 1.2703 2.5776 4 2>, LerpFactor=$User.VariantIndex)
    Lifetime=$System.CurrentLoopDuration
    Position Offset:NiagaraPosition=$User.CenterOffset
    Position Offset Coordinate Space=::Local
    Sprite Size Mode=::NewEnumerator3
    Uniform Sprite Size=&Lerp_Float(A=30, B=40, Alpha=$User.VariantIndex)
    UsePositionOffset=true
    Write Material Random=true
   )
   /SetVariables(Particles.FlareSize:Vector2f=&Multiply_Vector2D_ByFloat(Float=0.8, Vector2D=<100 100>)) {isSetParameters=true}
   &CameraOffset(Camera Offset Amount=&Lerp_Float(A=0, B=5, Alpha=$User.VariantIndex))
  ]
  particleUpdateScript [
   &ParticleState()
   &SpriteRotationRate(Rotation Rate=&Lerp_Float(A=50, B=-25, Alpha=&MakeFloatFromInt(Int=&ReturnExecIndex())))
   &FN_Light_Attributes(
    Color=$Particles.Color
    Create Color Attribute=true
    Create Exponent Attribute=true
    Create Radius Attribute=true
    Exponent=8
    Radius=80
    Scale Color=&Multiply_Float(A=0.2, B=1)
   )
   &ScaleColor(
    Scale Alpha=&FloatFromCurve(
     FloatCurve=DataInterface() {propertyValues {Curve {keys [
         {interpMode=RCIM_Cubic, tangentMode=RCTM_User, time=0, value=0}
         {arriveTangent=1.6667, leaveTangent=1.6667, time=0.1, value=1}
         {arriveTangent=-1.1111, interpMode=RCIM_Cubic, leaveTangent=-1.1111, tangentMode=RCTM_User, time=0.5, value=1}
         {arriveTangent=-5.2539, interpMode=RCIM_Cubic, leaveTangent=-5.2539, tangentMode=RCTM_User, time=1, value=0}
        ]}}}
     CurveIndex=$System.NormalizedLoopAge
     Scale Curve=0.5
    )
    Scale RGB=&Lerp_Vector(
     A=&VectorFromFloat(Value=&Multiply_Float(
       A=&FloatFromCurve(
        FloatCurve=DataInterface() {propertyValues {Curve {keys [
            {interpMode=RCIM_Cubic, tangentMode=RCTM_User, time=0, value=0}
            {arriveTangent=1.6667, leaveTangent=1.6667, time=0.1, value=1}
            {arriveTangent=-1.1111, interpMode=RCIM_Cubic, leaveTangent=-1.1111, tangentMode=RCTM_User, time=0.5, value=1}
            {arriveTangent=-5.2539, interpMode=RCIM_Cubic, leaveTangent=-5.2539, tangentMode=RCTM_User, time=1, value=0}
           ]}}}
        CurveIndex=$System.NormalizedLoopAge
        Scale Curve=0.5
       )
       B=&Waveform("Waveform [1]"=::NewEnumerator0, "[1] Frequency"=4, "[1] Amplitude Min/Max"=<0 1>, "[1] Scale"=0.4, "[1] Bias"=0.5)
      ))
     B=<1 1 1>
     Alpha=$User.VariantIndex
    )
    ScaleA=true
    ScaleRGB=true
   )
  ]
  rendererClasses [
   &NiagaraSpriteRendererProperties
   &NiagaraLightRendererProperties
   &NiagaraMeshRendererProperties
  ]
  renderers [
   {rendererClass=&NiagaraSpriteRendererProperties}
   {rendererClass=&NiagaraSpriteRendererProperties, rendererIndex=1}
   {rendererClass=&NiagaraLightRendererProperties, rendererIndex=2}
   {rendererClass=&NiagaraMeshRendererProperties, rendererIndex=3}
  ]
 }
 SmokePuff_Gen5 {
  emitterUpdateScript [
   &EmitterState() {module=EmitterState001}
   &SpawnBurst_Instantaneous(Use Spawn Probability=false, UseLoopCountLimit=false)
   &SpawnPerUnit(Module.Use Max Movement Threshold=true, Module.Use Movement Tolerance=false, Use Spawn Probability=false)
   &SpawnRate(
    SpawnRate=&FloatFromCurve(
     FloatCurve=DataInterface() {propertyValues {Curve {keys [
         {tangentMode=RCTM_User, time=0, value=1}
         {arriveTangent=-3.5689, interpMode=RCIM_Cubic, leaveTangent=-3.5689, tangentMode=RCTM_User, time=0.4, value=1}
         {arriveTangent=-1.8913, interpMode=RCIM_Cubic, leaveTangent=-1.8913, tangentMode=RCTM_User, time=0.65, value=0}
        ]}}}
     CurveIndex=$System.NormalizedLoopAge
     Scale Curve=8
    )
    Use Spawn Probability=false
   )
  ]
  particleSpawnScript [
   &PartitionParticles(Partition Mode:ENiagara_PartitionMode=::NewEnumerator4)
   &InitializeParticle(
    Color=&MultiplyLinearColor(
     Linear Color=&LerpLinearColors(StartColor=<rgba 6 1.7666 0.492 2>, EndColor=<rgba 3.822 4.5843 6 2>, LerpFactor=$User.VariantIndex)
     Scale Factor=&MakeVector4_FromLinearColor(LinearColor=&MakeCustomLinearColorFromBool(
       Bool=&IsPlatformSetActive(New Platform Set=DataInterface() {propertyValues {Platforms {qualityLevelMask=-4}}})
       True Color=&LerpLinearColors(
        StartColor=&FNTimeOfDayColor(
         Night Color=<rgba 0.6719 0.4385 0.3015 0.8>
         Morning Color=<rgba 0.5365 0.5365 0.5365 0.65>
         Day Color=<rgba 1 1 1 0.65>
         Evening Color=<rgba 0.375 0.375 0.375 0.65>
        )
        EndColor=&FNTimeOfDayColor(
         Night Color=<rgba 0.2357 1 0.7777 1>
         Morning Color=<rgba 0.5332 0.7357 0.8 1>
         Day Color=<rgba 1 1 1 0.65>
         Evening Color=<rgba 0 1 0.6659 1>
        )
        LerpFactor=$User.VariantIndex
       )
       False Color=&FNTimeOfDayColor(
        Night Color=<rgba 0.6094 0.4181 0.3443 1>
        Morning Color=<rgba 1 1 1 1>
        Day Color=<rgba 1 1 1 1>
        Evening Color=<rgba 0.3958 0.3958 0.3958 1>
       )
      ))
    )
    Lifetime=&RandomRangeFloat(Minimum=0.7, Maximum=0.8)
    Sprite Rotation Mode=::NewEnumerator2
    Sprite Size Mode=::NewEnumerator3
    Uniform Sprite Size=30
    UsePositionOffset=false
    Write Material Random=true
   ) {module=InitializeParticle002}
   &SkeletalMeshLocation(MaskSpawnGroup=false, Mesh Sampling Type=::NewEnumerator5, Skeletal Mesh=$System.Skeletal Mesh, Socket Sampling Mode=::NewEnumerator5)
   &AddVelocity(Velocity=<25 0 12>)
   &CameraOffset(Camera Offset Amount=&FN_FrontEndFloatModifier(Gameplay=0, Front End=0))
  ]
  particleUpdateScript [
   &ParticleState()
   &InheritSourceMovement(
    Applied Position Scale=&Multiply_VectorByFloat(
     Vector=<1 1 1>
     Float=&FloatFromCurve(FloatCurve=DataInterface() {propertyValues {Curve {keys [
          <at 0 1 RCIM_Cubic>
          {arriveTangent=-0.2, interpMode=RCIM_Cubic, leaveTangent=-0.2, time=0.25, value=1}
          {arriveTangent=-0.0269, interpMode=RCIM_Cubic, leaveTangent=-0.0269, tangentMode=RCTM_User, time=1, value=0.8}
         ]}}})
    )
    ApplyForce=false
    ApplyPosition=true
    ApplyPositionFalloff=false
   )
   &ScaleSpriteSize(Scale Factor=&Vector2DFromFloat(Value=&FloatFromCurve(
      FloatCurve=DataInterface() {propertyValues {Curve {keys [
          {arriveTangent=-0.0943, interpMode=RCIM_Cubic, leaveTangent=-0.0943, tangentMode=RCTM_User, time=0, value=0.2}
          {arriveTangent=0.116, interpMode=RCIM_Cubic, leaveTangent=0.116, tangentMode=RCTM_User, time=1, value=1}
         ]}}}
      Scale Curve=0.8
     )))
   &ScaleColor(
    Scale Alpha=&Lerp_Float(
     A=&Multiply_Float(
      A=&FloatFromCurve(FloatCurve=DataInterface() {propertyValues {Curve {keys [<at 0 0 RCIM_Cubic>,<at 0.2 0.5 RCIM_Cubic>,<at 1 0 RCIM_Cubic>]}}}, Scale Curve=3)
      B=&FloatFromCurve(
       FloatCurve=DataInterface() {propertyValues {Curve {keys [
           {tangentMode=RCTM_User, time=0, value=1}
           {arriveTangent=-3.5689, interpMode=RCIM_Cubic, leaveTangent=-3.5689, tangentMode=RCTM_User, time=0.65, value=1}
           {arriveTangent=-1.8913, interpMode=RCIM_Cubic, leaveTangent=-1.8913, tangentMode=RCTM_User, time=0.9, value=0}
          ]}}}
       CurveIndex=$System.NormalizedLoopAge
      )
     )
     B=&FloatFromCurve(FloatCurve=DataInterface() {propertyValues {Curve {keys [<at 0 0 RCIM_Cubic>,<at 0.2 0.5 RCIM_Cubic>,<at 1 0 RCIM_Cubic>]}}}, Scale Curve=3)
     Alpha=$User.VariantIndex
    )
    Scale RGB=&Lerp_Vector(
     A=&MakeVectorFromLinearColorRGB(Color=&LerpLinearColors(
       StartColor=$User.StartColor
       EndColor=<rgba 0.2031 0.2031 0.2031 1>
       LerpFactor=&FloatFromCurve(FloatCurve=DataInterface() {propertyValues {Curve {keys [
            {arriveTangent=-0.1505, interpMode=RCIM_Cubic, leaveTangent=-0.1505, tangentMode=RCTM_User, time=0, value=0}
            {arriveTangent=-0.1542, interpMode=RCIM_Cubic, leaveTangent=-0.1542, tangentMode=RCTM_User, time=0.7, value=1}
           ]}}})
      ))
     B=&MakeVectorFromLinearColorRGB(Color=&LerpLinearColors(
       StartColor=$User.StartColor
       EndColor=$User.EndColor
       LerpFactor=&FloatFromCurve(FloatCurve=DataInterface() {propertyValues {Curve {keys [
            {arriveTangent=-0.1505, interpMode=RCIM_Cubic, leaveTangent=-0.1505, tangentMode=RCTM_User, time=0, value=0}
            {arriveTangent=-0.1542, interpMode=RCIM_Cubic, leaveTangent=-0.1542, tangentMode=RCTM_User, time=0.7, value=1}
           ]}}})
      ))
     Alpha=$User.VariantIndex
    )
    ScaleA=true
    ScaleRGB=true
   ) {module=ScaleColor001}
   &DynamicMaterialParameters(
    Index 0 Param 1=&FloatFromCurve(
     FloatCurve=DataInterface() {propertyValues {Curve {keys [
         <at 0 0 RCIM_Cubic>
         {arriveTangent=1.5947, interpMode=RCIM_Cubic, leaveTangent=1.5947, tangentMode=RCTM_User, time=0.2557, value=0.5257}
         <at 1 1 RCIM_Cubic>
        ]}}}
     CurveIndex=$Particles.Age
     Scale Curve=0.25
    )
    Index 1 Param 1=1
    Index 1 Param 3=1
    Index 1 Param 4=0.5
    Index 2 Param 1=&FloatFromCurve(FloatCurve=DataInterface() {propertyValues {Curve {keys [
         {arriveTangent=1.9151, interpMode=RCIM_Cubic, leaveTangent=1.9151, tangentMode=RCTM_User, time=0, value=0}
         {interpMode=RCIM_Cubic, tangentMode=RCTM_User, time=1, value=1}
        ]}}})
    Index 2 Param 2=&FloatFromCurve(FloatCurve=DataInterface() {propertyValues {Curve {keys [<at 0 0.25 RCIM_Cubic>,<at 1 0 RCIM_Cubic>]}}})
    Index 2 Param 3=2
    Index 2 Param 4=0.7
    Module.Param0WriteEnabled=true
    Module.Param0WriteEnabled1=true
    Module.Param0WriteEnabled2=true
    Module.Param1WriteEnabled=true
    Module.Param1WriteEnabled1=true
    Module.Param1WriteEnabled2=true
    Module.Param2WriteEnabled=true
    Module.Param2WriteEnabled1=true
    Module.Param2WriteEnabled2=true
    Module.Param3WriteEnabled=false
    Module.Param3WriteEnabled1=true
    Module.Param3WriteEnabled2=true
    Write Parameter Index 1=true
    Write Parameter Index 2=true
   ) {module=DynamicMaterialParameters001}
   &CurlNoiseForce(
    Noise Frequency=250
    Noise Strength=&Multiply_Float(
     A=&FloatFromCurve(FloatCurve=DataInterface() {propertyValues {Curve {keys [<at 0 1>,<at 1 0>]}}}, Scale Curve=50)
     B=1
    )
    Pan Noise=false
   )
   &GravityForce(Gravity=&Multiply_VectorByFloat(
     Vector=<0 0 250>
     Float=&FloatFromCurve(FloatCurve=DataInterface() {propertyValues {Curve {keys [
          {interpMode=RCIM_Cubic, tangentMode=RCTM_User, time=0, value=0}
          {arriveTangent=2.3053, interpMode=RCIM_Cubic, leaveTangent=2.3053, tangentMode=RCTM_User, time=1, value=1}
         ]}}})
    ))
   &Drag(Drag=0.2, Use Linear Drag=true, Use Rotational Drag=true)
   &SolveForcesAndVelocity(Clamp Velocity=false, Limit Acceleration=false)
  ]
  rendererClasses [&NiagaraSpriteRendererProperties]
  renderers [
   {rendererClass=&NiagaraSpriteRendererProperties}
   {rendererClass=&NiagaraSpriteRendererProperties, rendererIndex=1}
  ]
 }
 SnowParticles {
  emitterUpdateScript [
   &EmitterState()
   &SpawnRate(
    SpawnRate=&FloatFromCurve(
     FloatCurve=DataInterface() {propertyValues {Curve {keys [
         {tangentMode=RCTM_User, time=0, value=1}
         {arriveTangent=-3.5689, interpMode=RCIM_Cubic, leaveTangent=-3.5689, tangentMode=RCTM_User, time=0.6, value=1}
         {arriveTangent=-1.8913, interpMode=RCIM_Cubic, leaveTangent=-1.8913, tangentMode=RCTM_User, time=1, value=0}
        ]}}}
     CurveIndex=$System.NormalizedLoopAge
     Scale Curve=8
    )
    Use Spawn Probability=false
   )
  ]
  particleSpawnScript [
   &InitializeParticle(
    Color=&RandomRangeLinearColor(Minimum=<rgba 1 1 1 1>, Maximum=$User.SnowBlue)
    Lifetime=&RandomRangeFloat(Minimum=0.5, Maximum=1)
    Sprite Rotation Mode=::NewEnumerator2
    Sprite Size Mode=::NewEnumerator1
    Sprite UV Mode=::NewEnumerator3
    UsePositionOffset=false
    Write Material Random=true
   )
   &SkeletalMeshLocation(MaskSpawnGroup=false, Mesh Sampling Type=::NewEnumerator5, Skeletal Mesh=$System.Skeletal Mesh, Socket Sampling Mode=::NewEnumerator5)
   &AddVelocity(Velocity=<25 0 12>)
   &ShapeLocation(MaskSpawnGroup=false, Override Seed=false, Sphere Radius=5, UseOwnerScale=true)
  ]
  particleUpdateScript [
   &ParticleState()
   &ScaleSpriteSize(Uniform Curve Sprite Scale=DataInterface() {propertyValues {Curve {keys [
        {arriveTangent=9.0725, interpMode=RCIM_Cubic, leaveTangent=9.0725, tangentMode=RCTM_User, time=0, value=0}
        {arriveTangent=-2.5249, interpMode=RCIM_Cubic, leaveTangent=-2.5249, tangentMode=RCTM_User, time=0.1408, value=0.9847}
        {arriveTangent=-0.0405, interpMode=RCIM_Cubic, leaveTangent=-0.0405, tangentMode=RCTM_User, time=1, value=0}
       ]}}})
   &ScaleColor(
    Scale Alpha=&FloatFromCurve(FloatCurve=DataInterface() {propertyValues {Curve {keys [
         {interpMode=RCIM_Cubic, tangentMode=RCTM_User, time=0, value=0}
         {interpMode=RCIM_Cubic, tangentMode=RCTM_User, time=0.5, value=1}
         {tangentMode=RCTM_User, time=1, value=0}
        ]}}})
    Scale RGB=&VectorFromFloat(Value=&FNTimeOfDayScalar(Night Scale=0.8, Morning Scale=1, Evening Scale=0.8))
    ScaleA=true
    ScaleRGB=true
   )
   &WindForce(Frequency=60, Offset By Time:Vector3f=<0 0 0>, Scale=1.5, Wind Speed=<0 0 -1>, Wind Speed Scale=20) {enabled=false}
   &AerodynamicDrag(
    Aerodynamic Drag:NiagaraFloat=&RandomRangeFloat(Minimum=5, Maximum=6)
    Aerodynamic Rotational Drag=0.2
    Drag Shape Curve=DataInterface() {propertyValues {Curve {keys [
        {interpMode=RCIM_Cubic, tangentMode=RCTM_User, time=0, value=0.1}
        {arriveTangent=2.0799, interpMode=RCIM_Cubic, leaveTangent=2.0799, tangentMode=RCTM_User, time=1, value=1}
       ]}}}
    Lift Shape Curve=DataInterface() {propertyValues {Curve {keys [
        {interpMode=RCIM_Cubic, tangentMode=RCTM_User, time=0, value=0}
        {interpMode=RCIM_Cubic, tangentMode=RCTM_User, time=0.5, value=1}
        {tangentMode=RCTM_User, time=1, value=0}
       ]}}}
    Use Linear Drag=true
    Use Rotational Drag=true
   ) {enabled=false}
   &Drag(Drag=0.8, Use Linear Drag=true, Use Rotational Drag=true)
   &CameraOffset(Camera Offset Amount=10)
   &GravityForce(Gravity=&Multiply_VectorByFloat(
     Vector=<0 0 250>
     Float=&FloatFromCurve(FloatCurve=DataInterface() {propertyValues {Curve {keys [
          {interpMode=RCIM_Cubic, tangentMode=RCTM_User, time=0, value=0}
          {arriveTangent=2.3053, interpMode=RCIM_Cubic, leaveTangent=2.3053, tangentMode=RCTM_User, time=1, value=1}
         ]}}})
    ))
   &InheritSourceMovement(
    Applied Position Scale=&Multiply_VectorByFloat(
     Vector=<1 1 1>
     Float=&FloatFromCurve(FloatCurve=DataInterface() {propertyValues {Curve {keys [
          <at 0 1 RCIM_Cubic>
          {arriveTangent=-0.2, interpMode=RCIM_Cubic, leaveTangent=-0.2, time=0.25, value=1}
          {arriveTangent=-0.0269, interpMode=RCIM_Cubic, leaveTangent=-0.0269, tangentMode=RCTM_User, time=1, value=0.8}
         ]}}})
    )
    ApplyForce=false
    ApplyPosition=true
    ApplyPositionFalloff=false
   ) {module=InheritSourceMovement001}
   &CurlNoiseForce(Noise Strength=1, Pan Noise=false)
   &SolveForcesAndVelocity(Clamp Velocity=false, Limit Acceleration=false)
  ]
  rendererClasses [&NiagaraSpriteRendererProperties]
  renderers [{rendererClass=&NiagaraSpriteRendererProperties}]
 }
 SnowParticles001 {
  emitterUpdateScript [
   &EmitterState()
   &SpawnRate(
    SpawnRate=&FloatFromCurve(
     FloatCurve=DataInterface() {propertyValues {Curve {keys [
         {tangentMode=RCTM_User, time=0, value=1}
         {arriveTangent=-3.5689, interpMode=RCIM_Cubic, leaveTangent=-3.5689, tangentMode=RCTM_User, time=0.6, value=1}
         {arriveTangent=-1.8913, interpMode=RCIM_Cubic, leaveTangent=-1.8913, tangentMode=RCTM_User, time=0.8, value=0}
        ]}}}
     CurveIndex=$System.NormalizedLoopAge
     Scale Curve=6
    )
    Use Spawn Probability=false
   )
  ]
  particleSpawnScript [
   &InitializeParticle(
    Color=&RandomRangeLinearColor(Minimum=<rgba 1 1 1 1>, Maximum=$User.SnowBlue)
    Lifetime=&RandomRangeFloat(Minimum=0.5, Maximum=1)
    Sprite Rotation Mode=::NewEnumerator2
    Sprite Size Mode=::NewEnumerator1
    Sprite UV Mode=::NewEnumerator3
    UsePositionOffset=false
    Write Material Random=true
   )
   &SkeletalMeshLocation(
    MaskSpawnGroup=false
    Mesh Sampling Type=::NewEnumerator5
    Sampled Position Offset=<2 0 0>
    Skeletal Mesh=$System.Skeletal Mesh
    Socket Sampling Mode=::NewEnumerator5
   )
   &AddVelocity(Velocity=<25 0 12>)
   &ShapeLocation(MaskSpawnGroup=false, Override Seed=false, Sphere Radius=2, UseOwnerScale=true)
  ]
  particleUpdateScript [
   &ParticleState()
   &ScaleSpriteSize(Uniform Curve Sprite Scale=DataInterface() {propertyValues {Curve {keys [
        {arriveTangent=9.0725, interpMode=RCIM_Cubic, leaveTangent=9.0725, tangentMode=RCTM_User, time=0, value=0}
        {arriveTangent=-2.5249, interpMode=RCIM_Cubic, leaveTangent=-2.5249, tangentMode=RCTM_User, time=0.2, value=0.9847}
        {arriveTangent=-0.0405, interpMode=RCIM_Cubic, leaveTangent=-0.0405, tangentMode=RCTM_User, time=1, value=0}
       ]}}})
   &ScaleColor(
    Scale Alpha=&FloatFromCurve(FloatCurve=DataInterface() {propertyValues {Curve {keys [
         {interpMode=RCIM_Cubic, tangentMode=RCTM_User, time=0, value=0}
         {interpMode=RCIM_Cubic, tangentMode=RCTM_User, time=0.5, value=1}
         {tangentMode=RCTM_User, time=1, value=0}
        ]}}})
    Scale RGB=&VectorFromFloat(Value=&FNTimeOfDayScalar(Night Scale=0.8, Morning Scale=1, Evening Scale=0.8))
    ScaleA=true
    ScaleRGB=true
   )
   &WindForce(Frequency=60, Offset By Time:Vector3f=<0 0 0>, Scale=1.5, Wind Speed=<0 0 -1>, Wind Speed Scale=20) {enabled=false}
   &AerodynamicDrag(
    Aerodynamic Drag:NiagaraFloat=&RandomRangeFloat(Minimum=5, Maximum=6)
    Aerodynamic Rotational Drag=0.2
    Drag Shape Curve=DataInterface() {propertyValues {Curve {keys [
        {interpMode=RCIM_Cubic, tangentMode=RCTM_User, time=0, value=0.1}
        {arriveTangent=2.0799, interpMode=RCIM_Cubic, leaveTangent=2.0799, tangentMode=RCTM_User, time=1, value=1}
       ]}}}
    Lift Shape Curve=DataInterface() {propertyValues {Curve {keys [
        {interpMode=RCIM_Cubic, tangentMode=RCTM_User, time=0, value=0}
        {interpMode=RCIM_Cubic, tangentMode=RCTM_User, time=0.5, value=1}
        {tangentMode=RCTM_User, time=1, value=0}
       ]}}}
    Use Linear Drag=true
    Use Rotational Drag=true
   ) {enabled=false}
   &Drag(Drag=0.8, Use Linear Drag=true, Use Rotational Drag=true)
   &CameraOffset()
   &GravityForce(Gravity=&Multiply_VectorByFloat(
     Vector=<0 0 250>
     Float=&FloatFromCurve(FloatCurve=DataInterface() {propertyValues {Curve {keys [
          {interpMode=RCIM_Cubic, tangentMode=RCTM_User, time=0, value=0}
          {arriveTangent=2.3053, interpMode=RCIM_Cubic, leaveTangent=2.3053, tangentMode=RCTM_User, time=1, value=1}
         ]}}})
    ))
   &InheritSourceMovement(
    Applied Position Scale=&Multiply_VectorByFloat(
     Vector=<1 1 1>
     Float=&FloatFromCurve(FloatCurve=DataInterface() {propertyValues {Curve {keys [
          <at 0 1 RCIM_Cubic>
          {arriveTangent=-0.2, interpMode=RCIM_Cubic, leaveTangent=-0.2, time=0.25, value=1}
          {arriveTangent=-0.0269, interpMode=RCIM_Cubic, leaveTangent=-0.0269, tangentMode=RCTM_User, time=1, value=0.8}
         ]}}})
    )
    ApplyForce=false
    ApplyPosition=true
    ApplyPositionFalloff=false
   ) {module=InheritSourceMovement001}
   &CurlNoiseForce(Noise Strength=1, Pan Noise=false)
   &SolveForcesAndVelocity(Clamp Velocity=false, Limit Acceleration=false)
  ]
  rendererClasses [&NiagaraSpriteRendererProperties, &NiagaraMeshRendererProperties]
  renderers [
   {rendererClass=&NiagaraSpriteRendererProperties}
   {rendererClass=&NiagaraMeshRendererProperties, rendererIndex=1}
  ]
 }
}
summary {
 systemName=NS_Backpack_Sunshine
 emitters [
  {emitterName=SmokePuff_Gen5, rendererClasses [&NiagaraSpriteRendererProperties]}
  {emitterName=SnowParticles, rendererClasses [&NiagaraSpriteRendererProperties]}
  {emitterName=Dry_Ice_Smoke_Exhaust, rendererClasses [&NiagaraSpriteRendererProperties]}
  {
   emitterName=LightAndGlow
   rendererClasses [
    &NiagaraSpriteRendererProperties
    &NiagaraLightRendererProperties
    &NiagaraMeshRendererProperties
   ]
  }
  {emitterName=Dry_Ice_Idle, rendererClasses [&NiagaraSpriteRendererProperties]}
  {
   emitterName=SnowParticles001
   rendererClasses [&NiagaraSpriteRendererProperties, &NiagaraMeshRendererProperties]
  }
 ]
 userVariables [
  User.bIsFrontEnd {type=NiagaraBool}
  User.bIsFrontEndPreview {type=NiagaraBool}
  User.bisSwinging {type=NiagaraBool}
  User.CenterOffset {type=NiagaraPosition, description="", defaultValue {}}
  User.Color_1 {type=LinearColor, description="", defaultValue {struct=LinearColor, value=<rgba 0.2219 0.2292 0.2193 1>}}
  User.Color_2 {type=LinearColor, description="", defaultValue {struct=LinearColor, value=<rgba 0.1302 0.1266 0.1196 1>}}
  User.EndColor {type=LinearColor, description="", defaultValue {struct=LinearColor, value=<rgba 0.0823 0.0823 0.0823 1>}}
  User.SnowBlue {type=LinearColor, description="", defaultValue {struct=LinearColor, value=<rgba 1 1 1 1>}}
  User.StartColor {type=LinearColor, description="", defaultValue {struct=LinearColor, value=<rgba 0.117 0.1746 0.1812 1>}}
  User.VariantIndex {type=NiagaraFloat, defaultValue=0, description=""}
 ]
}
