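# Anim Blueprint for the SunShine backpack cosmetic, animating its pack, stars and hanging chain with pose adjustments and spring physics.
# Linear-spring AnimDynamics on the stars form a base pose, then spine-driven pack rotation feeds three angular-spring chain bones with collision limits.
# The event graph derives rotation and alphas from spine deltas, plus movement flags.
# Cached poses are switched for skydiving and DBNO, which adds pack or star offsets.
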AnimBlueprint /BRCosmetics/Accessories/FORT_Backpacks/Backpack_M_MED_SunShine/Meshes/M_MED_SunShine_Pack_AnimBP
target_skeleton=&/M_MED_SunShine_Pack_Skeleton
graphs [
 AnimGraph {nodes [
   n1 = Anim Dynamics - Bone: dyn_chain_A!(ComponentPose, Alpha=1, GravityOverride:Vector="0,0,0", ExternalForce="0,0,0") -> (Pose) {
    angularBiasOverride=1
    angularDampingOverride=0.999
    angularSpringConstant=3000
    bAngularSpring=true
    bOverrideAngularDamping=true
    bUseGravityOverride=true
    simulationSpace=RootRelative
    alphaBoolBlend {blendOption=Linear}
    boundBone {boneName=dyn_chain_A}
    componentAppliedLinearAccClamp=<1500 1500 1200>
    componentLinearAccScale=<0.5 0.5 0.5>
    componentLinearVelScale=<0.6 0.6 0.6>
    physicsBodyDefinitions [{
      collisionType=CustomSphere
      sphereCollisionRadius=2.2
      boundBone {boneName=dyn_chain_A}
      boxExtents=<11 11 11>
      constraintSetup {angularLimitsMax=<0 15 55>, angularLimitsMin=<0 -45 0>, angularTarget=<0 0 0>}
      localJointOffset=<2 0 0>
     }]
    planarLimits [{drivingBone {boneName=pack}, planeTransform {location=<-0.534 4.7372 0>, rotation=<rot 0.0001 10 -90.0003>}}]
    sphericalLimits [
     {limitRadius=14, drivingBone {boneName=spine_01}, sphereLocalOffset=<0 0 0>}
     {limitRadius=14, drivingBone {boneName=spine_04}, sphereLocalOffset=<7.6385 0 0>}
     {limitRadius=13, drivingBone {boneName=pack}, sphereLocalOffset=<-18.4199 10.4376 -24.6877>}
     {limitRadius=15, drivingBone {boneName=pack}, sphereLocalOffset=<0 13.796 -16.2864>}
    ]
   }
   n2 = Anim Dynamics - Bone: dyn_chain_B!(ComponentPose=@n1, Alpha=1, GravityOverride:Vector="0,0,-980", ExternalForce="0,0,0") -> (Pose) {
    angularBiasOverride=1
    angularDampingOverride=0.9
    angularSpringConstant=3000
    bAngularSpring=true
    bOverrideAngularDamping=true
    bUseGravityOverride=true
    simulationSpace=RootRelative
    alphaBoolBlend {blendOption=Linear}
    boundBone {boneName=dyn_chain_B}
    componentAppliedLinearAccClamp=<1500 1500 1200>
    componentLinearAccScale=<0.5 0.5 0.5>
    componentLinearVelScale=<0.6 0.6 0.6>
    physicsBodyDefinitions [{
      collisionType=CustomSphere
      sphereCollisionRadius=3
      boundBone {boneName=dyn_chain_B}
      boxExtents=<10 10 10>
      constraintSetup {angularLimitsMax=<0 0 20>, angularLimitsMin=<0 -20 -5>, angularTarget=<0 0 0>}
      localJointOffset=<2 0 0>
     }]
    planarLimits [{drivingBone {boneName=pack}, planeTransform {location=<-0.6102 5.1694 0>, rotation=<rot 0.0001 10 -90.0003>}}]
   }
   Anim Dynamics - Bone: dyn_chain_C!(ComponentPose=@n2, Alpha=1, GravityOverride:Vector="0,0,0", ExternalForce="0,0,0") -> (Pose) {
    angularBiasOverride=1
    angularDampingOverride=0.85
    angularSpringConstant=3000
    bAngularSpring=true
    bOverrideAngularDamping=true
    bUseGravityOverride=true
    bUsePlanarLimit=false
    bUseSphericalLimits=true
    simulationSpace=RootRelative
    alphaBoolBlend {blendOption=Linear}
    boundBone {boneName=dyn_chain_C}
    componentAppliedLinearAccClamp=<1500 1500 1200>
    componentLinearAccScale=<0.5 0.5 0.5>
    componentLinearVelScale=<0.6 0.6 0.6>
    physicsBodyDefinitions [{
      collisionType=CustomSphere
      sphereCollisionRadius=3
      boundBone {boneName=dyn_chain_C}
      boxExtents=<10.5 10.5 10.5>
      constraintSetup {angularLimitsMax=<0 0 15>, angularLimitsMin=<0 -20 -5>, angularTarget=<0 0 0>}
      localJointOffset=<2 0 0>
     }]
    planarLimits [{drivingBone {boneName=pack}, planeTransform {location=<-0.4928 4.5037 0>, rotation=<rot 0.0001 10 -90.0003>}}]
    sphericalLimits [{limitRadius=14, drivingBone {boneName=spine_01}, sphereLocalOffset=<0 -2.6155 -1.4399>}]
   }
   n3 = "Transform (Modify) Bone - Bone: pack"(ComponentPose=Local To Component(LocalPose=Use cached pose 'OutputPose'()), Translation="0,-1,0", Rotation="0,0,15") {
    rotationMode=BMM_Additive
    rotationSpace=BCS_BoneSpace
    translationMode=BMM_Additive
    translationSpace=BCS_BoneSpace
    alphaBoolBlend {blendOption=Linear}
    boneToModify {boneName=pack}
    rotation {roll=15}
    translation {y=-1}
   }
   Output Pose(Result=Blend Poses by bool(BlendPose_0=Component To Local(ComponentPose=@n3), BlendPose_1=Use cached pose 'SavedPose'(), bActiveValue=Get bIsSkyDiving()) {blendType=Linear}) {layerGroup=None}
   n4 = "Transform (Modify) Bone - Bone: spine_01"(ComponentPose=Local To Component(LocalPose=Input Pose()), Alpha=1, Translation="0,0,0", Rotation="0,0,20", Scale="1,1,1") {
    rotationMode=BMM_Replace
    rotationSpace=BCS_WorldSpace
    scaleMode=BMM_Additive
    translationMode=BMM_Replace
    translationSpace=BCS_WorldSpace
    alphaBoolBlend {blendInTime=1, blendOption=Linear, blendOutTime=1}
    boneToModify {boneName=spine_01}
    rotation {roll=20}
   }
   n5 = "Transform (Modify) Bone - Bone: spine_04"(ComponentPose=@n4, Alpha=1, Translation="0,0,0", Rotation="0,0,20", Scale="1,1,1") {
    rotationMode=BMM_Replace
    rotationSpace=BCS_WorldSpace
    scaleMode=BMM_Additive
    translationMode=BMM_Replace
    translationSpace=BCS_WorldSpace
    alphaBoolBlend {blendInTime=1, blendOption=Linear, blendOutTime=1}
    boneToModify {boneName=spine_04}
    rotation {roll=20}
   }
   n6 = "Transform (Modify) Bone - Bone: dyn_star_center_mstr"!(ComponentPose=@n5, bAlphaBoolEnabled:bool=false, Translation="0,0,0", Rotation="0,0,-10", Scale="0.9,0.9,0.9") -> (Pose) {
    alphaInputType=Bool
    rotationMode=BMM_Additive
    rotationSpace=BCS_BoneSpace
    translationMode=BMM_Additive
    translationSpace=BCS_BoneSpace
    alphaBoolBlend {blendInTime=0.3, blendOption=Custom, blendOutTime=0.6}
    alphaScaleBiasClamp {bInterpResult=true}
    boneToModify {boneName=dyn_star_center_mstr}
    rotation {roll=-10}
    scale=<0.9 0.9 0.9>
   }
   n7 = "Transform (Modify) Bone - Bone: dyn_star_center_mstr"(ComponentPose=@n6, bAlphaBoolEnabled=false, Translation="0,0,0", Rotation="0,0,10", Scale="0.9,0.9,0.9") {
    alphaInputType=Bool
    rotationMode=BMM_Additive
    rotationSpace=BCS_BoneSpace
    translationMode=BMM_Additive
    translationSpace=BCS_BoneSpace
    alphaBoolBlend {blendInTime=0.3, blendOption=Custom, blendOutTime=0.6}
    alphaScaleBiasClamp {bInterpResult=true}
    boneToModify {boneName=dyn_star_center_mstr}
    rotation {roll=10}
    scale=<0.9 0.9 0.9>
   }
   n8 = Anim Dynamics - Bone: dyn_star_A(ComponentPose=@n7, Alpha=1, ExternalForce="0,0,0") {
    bLinearSpring=true
    bOverrideLinearDamping=true
    bUseGravityOverride=true
    linearDampingOverride=1
    linearSpringConstant=175
    simulationSpace=RootRelative
    alphaBoolBlend {blendOption=Linear}
    boundBone {boneName=dyn_star_A}
    componentAppliedLinearAccClamp=<2500 2500 2500>
    componentLinearAccScale=<0.2 0.2 0.2>
    componentLinearVelScale=<0.9 0.9 0.9>
    physicsBodyDefinitions [{
      boundBone {boneName=dyn_star_A}
      boxExtents=<60 60 60>
      constraintSetup {angularLimitsMax=<0 0 0>, angularLimitsMin=<0 0 0>, angularTarget=<0 0 0>, linearAxesMax=<3 0 0.5>, linearAxesMin=<-1 -1 -0.1>}
      localJointOffset=<0 0 0>
     }]
   }
   n9 = Anim Dynamics - Bone: dyn_star_B(ComponentPose=@n8, Alpha=1, ExternalForce="0,0,0") {
    bLinearSpring=true
    bOverrideLinearDamping=true
    bUseGravityOverride=true
    linearDampingOverride=1
    linearSpringConstant=175
    simulationSpace=RootRelative
    alphaBoolBlend {blendOption=Linear}
    boundBone {boneName=dyn_star_B}
    componentAppliedLinearAccClamp=<2500 2500 2500>
    componentLinearAccScale=<0.2 0.2 0.2>
    componentLinearVelScale=<0.9 0.9 0.9>
    physicsBodyDefinitions [{
      boundBone {boneName=dyn_star_B}
      boxExtents=<60 60 60>
      constraintSetup {angularLimitsMax=<0 0 0>, angularLimitsMin=<0 0 0>, angularTarget=<0 0 0>, linearAxesMax=<2 0 0>, linearAxesMin=<0 0 -0.8>}
      localJointOffset=<0 0 0>
     }]
   }
   n10 = Anim Dynamics - Bone: dyn_star_C(ComponentPose=@n9, Alpha=1, ExternalForce="0,0,0") {
    bLinearSpring=true
    bOverrideLinearDamping=true
    bUseGravityOverride=true
    linearDampingOverride=1
    linearSpringConstant=175
    simulationSpace=RootRelative
    alphaBoolBlend {blendOption=Linear}
    boundBone {boneName=dyn_star_C}
    componentAppliedLinearAccClamp=<2500 2500 2500>
    componentLinearAccScale=<0.2 0.2 0.2>
    componentLinearVelScale=<0.9 0.9 0.9>
    physicsBodyDefinitions [{
      boundBone {boneName=dyn_star_C}
      boxExtents=<60 60 60>
      constraintSetup {angularLimitsMax=<0 0 0>, angularLimitsMin=<0 0 0>, angularTarget=<0 0 0>, linearAxesMax=<1 0 1>, linearAxesMin=<-2 -2 -2>}
      localJointOffset=<0 0 0>
     }]
   }
   Save cached pose 'BasePose'(Pose=Component To Local(ComponentPose=@n10)) {cacheName=BasePose}
   n11 = "Transform (Modify) Bone - Bone: pack"(ComponentPose=Local To Component(LocalPose=Use cached pose 'BasePose'()), Translation="2,0,0", Rotation="0,0,10") {
    rotationMode=BMM_Additive
    rotationSpace=BCS_BoneSpace
    translationMode=BMM_Additive
    translationSpace=BCS_BoneSpace
    alphaBoolBlend {blendInTime=0.1, blendOption=Custom, blendOutTime=0.6}
    alphaScaleBiasClamp {bInterpResult=true}
    boneToModify {boneName=pack}
    rotation {roll=10}
    translation {x=2}
   }
   n12 = "Transform (Modify) Bone - Bone: pack"(ComponentPose=@n11, Translation="0,0,0", Rotation=Get backpack_rotate()) {
    rotationMode=BMM_Additive
    scaleMode=BMM_Additive
    scaleSpace=BCS_BoneSpace
    translationMode=BMM_Additive
    alphaBoolBlend {blendOption=Linear}
    boneToModify {boneName=pack}
   }
   n13 = "Transform (Modify) Bone - Bone: pack"(ComponentPose=@n12, Alpha=Get Spine Alpha Z() -> (Spine Alpha Z:float), Translation="0,0,0", Rotation="0,0,15") {
    rotationMode=BMM_Additive
    rotationSpace=BCS_BoneSpace
    translationMode=BMM_Additive
    translationSpace=BCS_BoneSpace
    alphaBoolBlend {blendInTime=0.1, blendOption=Custom, blendOutTime=0.6}
    alphaScaleBiasClamp {bInterpResult=true, interpSpeedIncreasing=2}
    boneToModify {boneName=pack}
    rotation {roll=15}
   }
   n14 = "Transform (Modify) Bone - Bone: pack"(ComponentPose=@n13, Alpha=Get Spine Alpha X() -> (Spine Alpha X:float), Translation="0,0,0", Rotation="0,0,25") {
    rotationMode=BMM_Additive
    rotationSpace=BCS_BoneSpace
    translationMode=BMM_Additive
    translationSpace=BCS_BoneSpace
    alphaBoolBlend {blendInTime=0.1, blendOption=Custom, blendOutTime=0.6}
    alphaScaleBiasClamp {bInterpResult=true, interpSpeedIncreasing=2}
    boneToModify {boneName=pack}
    rotation {roll=25}
   }
   n15 = Anim Dynamics - Bone: dyn_chain_A(ComponentPose=@n14, Alpha=1, GravityOverride="0,0,0", ExternalForce="0,0,0") {
    angularBiasOverride=1
    angularDampingOverride=0.999
    angularSpringConstant=3000
    bAngularSpring=true
    bOverrideAngularDamping=true
    bUseGravityOverride=true
    simulationSpace=RootRelative
    alphaBoolBlend {blendOption=Linear}
    boundBone {boneName=dyn_chain_A}
    componentAppliedLinearAccClamp=<1500 1500 1200>
    componentLinearAccScale=<0.5 0.5 0.5>
    componentLinearVelScale=<0.6 0.6 0.6>
    physicsBodyDefinitions [{
      collisionType=CustomSphere
      sphereCollisionRadius=2.2
      boundBone {boneName=dyn_chain_A}
      boxExtents=<11 11 11>
      constraintSetup {angularLimitsMax=<0 15 55>, angularLimitsMin=<0 -45 0>, angularTarget=<0 0 0>}
      localJointOffset=<2 0 0>
     }]
    planarLimits [{drivingBone {boneName=pack}, planeTransform {location=<-0.534 4.7372 0>, rotation=<rot 0.0001 10 -90.0003>}}]
    sphericalLimits [
     {limitRadius=14, drivingBone {boneName=spine_01}, sphereLocalOffset=<0 0 0>}
     {limitRadius=14, drivingBone {boneName=spine_04}, sphereLocalOffset=<7.6385 0 0>}
     {limitRadius=13, drivingBone {boneName=pack}, sphereLocalOffset=<-18.4199 10.4376 -24.6877>}
     {limitRadius=15, drivingBone {boneName=pack}, sphereLocalOffset=<0 13.796 -16.2864>}
    ]
   }
   n16 = Anim Dynamics - Bone: dyn_chain_B(ComponentPose=@n15, Alpha=1, GravityOverride="0,0,-980", ExternalForce="0,0,0") {
    angularBiasOverride=1
    angularDampingOverride=0.9
    angularSpringConstant=3000
    bAngularSpring=true
    bOverrideAngularDamping=true
    bUseGravityOverride=true
    simulationSpace=RootRelative
    alphaBoolBlend {blendOption=Linear}
    boundBone {boneName=dyn_chain_B}
    componentAppliedLinearAccClamp=<1500 1500 1200>
    componentLinearAccScale=<0.5 0.5 0.5>
    componentLinearVelScale=<0.6 0.6 0.6>
    physicsBodyDefinitions [{
      collisionType=CustomSphere
      sphereCollisionRadius=3
      boundBone {boneName=dyn_chain_B}
      boxExtents=<10 10 10>
      constraintSetup {angularLimitsMax=<0 0 20>, angularLimitsMin=<0 -20 -5>, angularTarget=<0 0 0>}
      localJointOffset=<2 0 0>
     }]
    planarLimits [{drivingBone {boneName=pack}, planeTransform {location=<-0.6102 5.1694 0>, rotation=<rot 0.0001 10 -90.0003>}}]
   }
   n17 = Anim Dynamics - Bone: dyn_chain_C(ComponentPose=@n16, Alpha=1, GravityOverride="0,0,0", ExternalForce="0,0,0") {
    angularBiasOverride=1
    angularDampingOverride=0.85
    angularSpringConstant=3000
    bAngularSpring=true
    bOverrideAngularDamping=true
    bUseGravityOverride=true
    bUsePlanarLimit=false
    simulationSpace=RootRelative
    alphaBoolBlend {blendOption=Linear}
    boundBone {boneName=dyn_chain_C}
    componentAppliedLinearAccClamp=<1500 1500 1200>
    componentLinearAccScale=<0.5 0.5 0.5>
    componentLinearVelScale=<0.6 0.6 0.6>
    physicsBodyDefinitions [{
      collisionType=CustomSphere
      sphereCollisionRadius=3
      boundBone {boneName=dyn_chain_C}
      boxExtents=<10.5 10.5 10.5>
      constraintSetup {angularLimitsMax=<0 0 15>, angularLimitsMin=<0 -20 -5>, angularTarget=<0 0 0>}
      localJointOffset=<2 0 0>
     }]
    planarLimits [{drivingBone {boneName=pack}, planeTransform {location=<-0.4928 4.5037 0>, rotation=<rot 0.0001 10 -90.0003>}}]
    sphericalLimits [{limitRadius=14, drivingBone {boneName=spine_01}, sphereLocalOffset=<0 -2.6155 -1.4399>}]
   }
   Save cached pose 'OutputPose'(Pose=Component To Local(ComponentPose=@n17)) {cacheName=OutputPose}
   n18 = "Transform (Modify) Bone - Bone: dyn_star_center_mstr"!(
    ComponentPose=Local To Component(LocalPose=Use cached pose 'BasePose'())
    Alpha:float=1
    Translation="0,0,-2"
    Rotation="0,0,0"
    Scale="1,1,1"
   ) -> (Pose) {
    rotationMode=BMM_Additive
    rotationSpace=BCS_BoneSpace
    scaleMode=BMM_Additive
    scaleSpace=BCS_BoneSpace
    translationMode=BMM_Additive
    translationSpace=BCS_BoneSpace
    alphaBoolBlend {blendOption=Linear}
    boneToModify {boneName=dyn_star_center_mstr}
    translation {z=-2}
   }
   n19 = "Transform (Modify) Bone - Bone: dyn_star_B_mstr"(ComponentPose=@n18, Alpha=1, Translation="0,-2,-2", Rotation="0,0,0", Scale="1,1,1") {
    rotationMode=BMM_Additive
    rotationSpace=BCS_BoneSpace
    scaleMode=BMM_Additive
    scaleSpace=BCS_BoneSpace
    translationMode=BMM_Additive
    translationSpace=BCS_BoneSpace
    alphaBoolBlend {blendOption=Linear}
    boneToModify {boneName=dyn_star_B_mstr}
    translation {y=-2, z=-2}
   }
   Save cached pose 'SavedPose'(Pose=Blend Poses by bool(BlendPose_0=Component To Local(ComponentPose=@n19), BlendPose_1=Use cached pose 'OutputPose'(), bActiveValue=Get bIsDBNO()) {blendType=Linear}) {cacheName=SavedPose}
  ]}
 EventGraph {nodes [
   Event BlueprintUpdateAnimation()
   > Is Valid(InputObject=Get BodyMesh() -> (BodyMesh:Skeletal Mesh Component Object Reference))
   > n1 = Gravity Override State Function DEPRECATED(CharacterPartAnimInstance=Self-Reference(), GameTimeInSeconds=GetGameTimeInSeconds(), GravityOverrideParams=Get Gravity Override Params())
   > Set Gravity Override(Gravity Override=@n1)
   n2 = BreakTransform(InTransform=GetSocketTransform(self=Get BodyMesh(), InSocketName=pelvis, TransformSpace=RTS_World))
   > Set spine_01 translation(spine_01 translation:Vector=@n2) -> (Output_Get:Vector)
   > Set spine_01rotation(spine_01rotation:Rotator=@n2.Rotation) -> (Output_Get:Rotator)
   n3 = BreakTransform(InTransform=GetSocketTransform(self=Get BodyMesh(), InSocketName=spine_02, TransformSpace=RTS_World))
   > Set spine 04 translation(spine 04 translation:Vector=@n3) -> (Output_Get:Vector)
   > Set spine 04 rotation(spine 04 rotation:Rotator=@n3.Rotation) -> (Output_Get:Rotator)
   n4 = Get bIsJumping()
   n5 = Get bIsFalling()
   > "Set  jump alpha pack"(" jump alpha pack":double=SelectFloat(A=1, B=0, bPickA=OR Boolean(A=@n4, B=@n5))) -> (Output_Get:double)
   n6 = Get PawnMovementDirectionAzimuth() -> (PawnMovementDirectionAzimuth:float)
   n7 = AND Boolean!(
    A=AND Boolean!(A="float > float"(A=@n6, B=-45), B="float < float"(A=@n6, B=45)) -> (ReturnValue)
    B=Get bIsMoving2D()
    C:bool=NOT Boolean(A=Get bIsCrouching())
   ) -> (ReturnValue)
   > Set Is moving forward(Is moving forward:bool=@n7) -> (Output_Get:bool)
   > Set Fwd_Mov(Fwd_Mov:double="To Float (Boolean)"(InBool=@n7)) -> (Output_Get:double)
   n8 = Get bIsCrouching()
   n9 = OR Boolean!(A=Get bIsSkydiveFloating(), B=Get bIsSkyDiving(), C:bool=Get bIsTargeting(), D:bool=Get bIsJumping(), E:bool=Get bIsFalling()) -> (ReturnValue)
   n10 = Get PawnMovementDirectionAzimuth()
   n11 = "float > float"(A=@n10, B=25)
   n12 = "float < float"(A=@n10, B=140)
   > Set Right mvmnt corr(Right mvmnt corr:bool=AND Boolean!(A=AND Boolean!(A=@n11, B=@n12) -> (ReturnValue), B=Get bIsMoving2D(), C:bool=NOT Boolean(A=@n8), D:bool=NOT Boolean(A=@n9)) -> (ReturnValue)) -> (Output_Get:bool)
   n13 = Get bIsCrouching()
   n14 = OR Boolean(A=Get bIsSkydiveFloating(), B=Get bIsSkyDiving(), C=Get bIsTargeting(), D=Get bIsJumping(), E=Get bIsFalling())
   n15 = Get PawnMovementDirectionAzimuth()
   n16 = "float > float"(A=@n15, B=-140)
   n17 = "float < float"(A=@n15, B=-25)
   > Set Left mvmnt corr(Left mvmnt corr:bool=AND Boolean!(A=AND Boolean!(A=@n16, B=@n17) -> (ReturnValue), B=Get bIsMoving2D(), C:bool=NOT Boolean(A=@n13), D:bool=NOT Boolean(A=@n14)) -> (ReturnValue)) -> (Output_Get:bool)
   n18 = Get bIsCrouching()
   n19 = OR Boolean(A=Get bIsSkydiveFloating(), B=Get bIsSkyDiving(), C=Get bIsTargeting(), D=Get bIsJumping(), E=Get bIsFalling())
   n20 = Get PawnMovementDirectionAzimuth()
   n21 = AND Boolean(A="float > float"(A=@n20, B=140), B="float < float"(A=@n20, B=180))
   n22 = AND Boolean(A="float > float"(A=@n20, B=-180), B="float < float"(A=@n20, B=-140))
   > Set Back mvmnt corr(Back mvmnt corr:bool=AND Boolean!(A=OR Boolean!(A=@n21, B=@n22) -> (ReturnValue), B=Get bIsMoving2D(), C:bool=NOT Boolean(A=@n18), D:bool=NOT Boolean(A=@n19)) -> (ReturnValue)) -> (Output_Get:bool)
   n23 = Get BodyMesh()
   n24 = NOT Boolean(A=Get bIsSkyDiving())
   n25 = NOT Boolean(A=Get bIsTargeting())
   n26 = NOT Boolean(A=Get bIsFalling())
   n27 = NOT Boolean(A=Get bIsMoving2D())
   n28 = NOT Boolean(A=Get bIsCrouching())
   n29 = NOT Boolean(A=Get bIsSkydiveFloating())
   n30 = NOT Boolean(A=Get bIsJumping())
   > Set Spine Alpha Z(Spine Alpha Z:float=SelectFloat(
     A=GetAlphaFromReferenceTransformDelta(SkeletalMeshComponent=@n23, BoneName=spine_01, BaseName=pelvis, InputRange=(X=0,Y=10), TwistAxis="1,0,0")
     B=0
     bPickA=OR Boolean!(A=@n27, B=@n28, C:bool=@n29, D:bool=@n24, E:bool=@n25, F:bool=@n30, G:bool=@n26) -> (ReturnValue)
    )) -> (Output_Get:float)
   n31 = Get BodyMesh()
   n32 = NOT Boolean(A=Get bIsMoving2D())
   n33 = NOT Boolean(A=Get bIsCrouching())
   n34 = NOT Boolean(A=Get bIsSkydiveFloating())
   n35 = NOT Boolean(A=Get bIsSkyDiving())
   n36 = NOT Boolean(A=Get bIsTargeting())
   n37 = NOT Boolean(A=Get bIsJumping())
   n38 = NOT Boolean(A=Get bIsFalling())
   > Set Spine Alpha X(Spine Alpha X:float=SelectFloat(
     A=GetAlphaFromReferenceTransformDelta(SkeletalMeshComponent=@n31, BoneName=spine_03, BaseName=spine_02, DeltaComponentType=EulerX, InputRange=(X=0,Y=-10), TwistAxis="1,0,0")
     B=0
     bPickA=OR Boolean(A=@n32, B=@n33, C=@n34, D=@n35, E=@n36, F=@n37, G=@n38)
    )) -> (Output_Get:float)
   n39 = Get BodyMesh()
   n40 = NOT Boolean(A=Get bIsMoving2D())
   n41 = NOT Boolean(A=Get bIsCrouching())
   n42 = NOT Boolean(A=Get bIsSkydiveFloating())
   n43 = NOT Boolean(A=Get bIsSkyDiving())
   n44 = NOT Boolean(A=Get bIsTargeting())
   n45 = NOT Boolean(A=Get bIsJumping())
   n46 = NOT Boolean(A=Get bIsFalling())
   > Set Spine Alpha 3(Spine Alpha 3:float=SelectFloat(
     A=GetAlphaFromReferenceTransformDelta(SkeletalMeshComponent=@n39, BoneName=spine_04, BaseName=spine_02, InputRange=(X=0,Y=-5), TwistAxis="1,0,0")
     B=0
     bPickA=OR Boolean(A=@n40, B=@n41, C=@n42, D=@n43, E=@n44, F=@n45, G=@n46)
    )) -> (Output_Get:float)
   n47 = GetOwnerSkeletalMeshComponent()
   n48 = BreakRotator(InRot=BreakTransform(InTransform=GetDeltaTransformFromRefPose(self=@n47, BoneName=spine_05, BaseName=spine_01)).Rotation)
   n49 = MapRangeClamped(Value=@n48.Pitch, InRangeA=30, InRangeB=-30, OutRangeA=-30, OutRangeB=30)
   n50 = MapRangeClamped(Value=@n48.Pitch, InRangeA=0, InRangeB=-30, OutRangeA=-5, OutRangeB=25)
   n51 = OR Boolean!(A=Get bIsSkydiveFloating(), B=Get bIsSkyDiving()) -> (ReturnValue)
   > Set backpack_rotate(backpack_rotate=MakeRotator(Roll=MapRangeClamped(Value=@n48.Yaw, InRangeA=0, InRangeB=-30, OutRangeA=5, OutRangeB=25), Pitch=SelectFloat(A=@n50, B=@n49, bPickA=@n51)))
   n52 = Event OnCharacterCustomizationLoaded()
   > Set BodyMesh(BodyMesh=Get Array Item with EFort Custom Part Type(Array=@n52.CustomMeshParts))
   TryGetPawnOwner()
  ]}
]
variables [
 " jump alpha pack" {type=number}
 Back mvmnt corr {type=boolean}
 backpack_rotate {type=Rotator}
 backpack_translation {type=Vector}
 BodyMesh {type=SkeletalMeshComponent}
 Fwd_Mov {type=number}
 Gravity Override {type=Vector}
 Gravity Override Params {type=GravityOverrideParamsStruct}
 Is moving forward {type=boolean}
 Left mvmnt corr {type=boolean}
 Right mvmnt corr {type=boolean}
 spine 04 rotation {type=Rotator}
 spine 04 translation {type=Vector}
 Spine Alpha 3 {type=number}
 Spine Alpha X {type=number}
 Spine Alpha Z {type=number}
 spine_01 translation {type=Vector}
 spine_01rotation {type=Rotator}
]
